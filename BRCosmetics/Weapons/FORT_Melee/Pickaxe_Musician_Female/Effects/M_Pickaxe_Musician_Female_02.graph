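# Unlit, audio-reactive effect material for the female Musician pickaxe, enabled for Niagara mesh particles.
# MPC Amplitude_Master is remapped to scale MainColor, which is multiplied by a time-of-day scalar and passed through MF_FX_EmissiveAdjustments.
# A second branch blends red and green vertex-color heights, scales them by amplitude and transforms that local Z vector, apparently as a vertex offset.
# Separate debug-sine and repeating UV gradient chains feed nothing visible in the graph.
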Material /BRCosmetics/Weapons/FORT_Melee/Pickaxe_Musician_Female/Effects/M_Pickaxe_Musician_Female_02
graph {
 properties {
  bCastRayTracedShadows=false
  bUsedWithNiagaraMeshParticles=true
  pixelDepthOffsetMode=PDOM_Legacy
  shadingModel=MSM_Unlit
  usedShadingModels=Unlit
 }
 nodes [
  Constant()
  n1 = Constant() {r=0.25}
  n2 = Constant() {r=1}
  n3 = VectorParameter() {parameterName=MainColor, defaultValue=<rgba 3 1.6915 0 0>}
  n4 = CollectionParameter() {collection=/Game/Athena/Events/AudioAnalysis/MPC_Audio_Reactive_Parameters, parameterName=Amplitude_Master}
  n5 = MaterialFunctionCall(
   Input=MaterialFunctionCall(R=@n4, G=@n4, B=@n4) {materialFunction=&AppendMany}.RGB
   Input Low=Constant() {r=0.05}
   Input High=Constant() {r=1}
   Target Low=Constant() {r=1}
   Target High=ScalarParameter() {defaultValue=2, parameterName=MaxColorRemap}
  ) {materialFunction=&RemapValueRange}
  n6 = MaterialFunctionCall(R=@n5.Result, G=@n5.Result, B=@n5.Result) {materialFunction=&AppendMany}
  MaterialFunctionCall(Emissive=Multiply(
    A=MaterialFunctionCall(Value_Morning=@n1, Value_Day=@n2, Value_Evening=@n1, Value_Night=@n2) {materialFunction=&Scalar_Time_of_Day}.Result
    B=Multiply(A=@n3.RGB, B=@n6.RGB)
   )) {materialFunction=&MF_FX_EmissiveAdjustments}
  MaterialFunctionCall(
   Input=MaterialFunctionCall(Speed=ScalarParameter() {defaultValue=2, parameterName=DEBUG_SineSpeed}) {materialFunction=&DebugTimeSine}.Result
   Input Low=Constant()
   Input High=Constant() {r=1}
   Target Low=ScalarParameter() {parameterName=Param_1}
   Target High=Constant() {r=1}
  ) {materialFunction=&RemapValueRange}
  n7 = TextureCoordinate()
  n8 = Multiply(A=ScalarParameter() {parameterName=Param_3}) {constB=0.1}
  n9 = Multiply(A=ComponentMask(None=Add(A=@n7, B=@n8)) {r=true}, B=ScalarParameter() {defaultValue=32, parameterName=NumberOfGradients})
  n10 = Saturate(None=ScalarParameter() {defaultValue=1, parameterName=Param_2})
  n11 = Subtract(A=Saturate(None=Frac(None=@n9)), B=@n10)
  n12 = Subtract(B=@n10)
  n13 = Divide(B=@n10)
  n14 = Max(A=Multiply(A=@n11, B=Divide(B=@n12)), B=Multiply(A=Multiply(A=@n13, B=@n11)) {constB=-1})
  OneMinus(None=Power(Base=Saturate(None=@n14), Exp=ScalarParameter() {defaultValue=1, parameterName=Power}))
  ScalarParameter() {defaultValue=-20, parameterName=Param_5}
  n15 = VertexColor()
  n16 = Add(
   A=LinearInterpolate(B=ScalarParameter() {defaultValue=30, parameterName=RedHeight}, Alpha=@n15.R)
   B=LinearInterpolate(B=ScalarParameter() {defaultValue=-20, parameterName=GreenHeight}, Alpha=@n15.G)
  )
  n17 = MaterialFunctionCall(
   Input=CollectionParameter() {collection=&MPC_Audio_Reactive_Parameters, parameterName=Amplitude_Master}
   Input Low=Constant()
   Input High=Constant() {r=1}
   Target Low=Constant()
   Target High=Constant() {r=1}
  ) {materialFunction=&RemapValueRange}
  Transform(None=MaterialFunctionCall(B=Multiply(A=@n16, B=@n17.Result)) {materialFunction=&AppendMany}.RGB) {transformSourceType=TRANSFORMSOURCE_Local}
 ]
}
