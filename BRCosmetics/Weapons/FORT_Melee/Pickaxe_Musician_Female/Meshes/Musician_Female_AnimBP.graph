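# Animation Blueprint for the Musician Female pickaxe cosmetic, animating its orbiting planets and UFO procedurally.
# AnimDynamics adds springy motion to three planets, then the cached pose feeds a chain of additive Modify Bone nodes.
# The update event derives spins, UFO bob and banking from real time.
# Delayed random rolls, re-evaluated for each pose, toggle three UFO tilt offsets that blend through interpolated alphas and can overlap.
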AnimBlueprint /BRCosmetics/Weapons/FORT_Melee/Pickaxe_Musician_Female/Meshes/Musician_Female_AnimBP
target_skeleton=&/Musician_Female_Skeleton
graphs [
 AnimGraph {nodes [
   n1 = "Transform (Modify) Bone - Bone: dyn_UFO_jnt"(
    ComponentPose=Local To Component(LocalPose=Use cached pose 'PlanetRotations'())
    Alpha=1
    Translation=Get TranslateUFO() -> (TranslateUFO:Vector)
    Rotation=Get UFObanking() -> (UFObanking:Rotator)
    Scale="1,1,1"
   ) {
    rotationMode=BMM_Additive
    rotationSpace=BCS_BoneSpace
    translationMode=BMM_Additive
    translationSpace=BCS_BoneSpace
    alphaBoolBlend {blendOption=Linear}
    boneToModify {boneName=dyn_UFO_jnt}
   }
   n2 = "Transform (Modify) Bone - Bone: dyn_UFO_jnt"(ComponentPose=@n1, Alpha=1, Translation="0,0,0", Rotation=Get RotateUFO() -> (RotateUFO:Rotator), Scale="1,1,1") {
    rotationMode=BMM_Additive
    rotationSpace=BCS_BoneSpace
    translationMode=BMM_Additive
    translationSpace=BCS_BoneSpace
    alphaBoolBlend {blendOption=Linear}
    boneToModify {boneName=dyn_UFO_jnt}
   }
   n3 = "Transform (Modify) Bone - Bone: dyn_planetA_rotator_jnt"(ComponentPose=@n2, Alpha=1, Translation="0,0,0", Rotation=Get RotatePlanet1() -> (RotatePlanet1:Rotator), Scale="1,1,1") {rotationMode=BMM_Additive, alphaBoolBlend {blendOption=Linear}, boneToModify {boneName=dyn_planetA_rotator_jnt}}
   n4 = Get RotatePlanet2() -> (RotatePlanet2:Rotator)
   n5 = Get RotatePlanet1()
   n6 = "Transform (Modify) Bone - Bone: dyn_planetA_jnt"(
    ComponentPose="Transform (Modify) Bone - Bone: dyn_planetC_rotator_jnt"(
     ComponentPose="Transform (Modify) Bone - Bone: dyn_planetB_rotator_jnt"(ComponentPose=@n3, Alpha=1, Translation="0,0,0", Rotation=@n4, Scale="1,1,1") {rotationMode=BMM_Additive, alphaBoolBlend {blendOption=Linear}, boneToModify {boneName=dyn_planetB_rotator_jnt}}
     Alpha=1
     Translation="0,0,0"
     Rotation=Get RotatePlanet3() -> (RotatePlanet3:Rotator)
     Scale="1,1,1"
    ) {rotationMode=BMM_Additive, alphaBoolBlend {blendOption=Linear}, boneToModify {boneName=dyn_planetC_rotator_jnt}}
    Alpha=1
    Translation="0,0,0"
    Rotation=@n5
    Scale="1,1,1"
   ) {rotationMode=BMM_Additive, alphaBoolBlend {blendOption=Linear}, boneToModify {boneName=dyn_planetA_jnt}}
   n7 = "Transform (Modify) Bone - Bone: dyn_UFO_rotator_jnt"(
    ComponentPose="Transform (Modify) Bone - Bone: dyn_planetC_jnt"(
     ComponentPose="Transform (Modify) Bone - Bone: dyn_planetB_jnt"(ComponentPose=@n6, Alpha=1, Translation="0,0,0", Rotation=@n5, Scale="1,1,1") {rotationMode=BMM_Additive, alphaBoolBlend {blendOption=Linear}, boneToModify {boneName=dyn_planetB_jnt}}
     Alpha=1
     Translation="0,0,0"
     Rotation=@n5
     Scale="1,1,1"
    ) {rotationMode=BMM_Additive, alphaBoolBlend {blendOption=Linear}, boneToModify {boneName=dyn_planetC_jnt}}
    Alpha=Get UFOpose1() -> (UFOpose1:double)
    Translation="0,0,0"
    Rotation="0,20,60"
    Scale="1,1,1"
   ) {
    rotationMode=BMM_Additive
    alphaBoolBlend {blendOption=Linear}
    alphaScaleBiasClamp {bClampResult=true, bInterpResult=true, bias=0.1, interpSpeedDecreasing=1, interpSpeedIncreasing=1}
    boneToModify {boneName=dyn_UFO_rotator_jnt}
    rotation {roll=60, yaw=20}
   }
   n8 = "Transform (Modify) Bone - Bone: dyn_UFO_rotator_jnt"(ComponentPose=@n7, Alpha=Get UFOpose2() -> (UFOpose2:double), Translation="0,0,0", Rotation="40,0,-90", Scale="1,1,1") {
    rotationMode=BMM_Additive
    alphaBoolBlend {blendOption=Linear}
    alphaScaleBiasClamp {bClampResult=true, bInterpResult=true, bias=0.1, interpSpeedDecreasing=1, interpSpeedIncreasing=1}
    boneToModify {boneName=dyn_UFO_rotator_jnt}
    rotation {pitch=40, roll=-90}
   }
   n9 = "Transform (Modify) Bone - Bone: dyn_UFO_rotator_jnt"(ComponentPose=@n8, Alpha=Get UFOpose3() -> (UFOpose3:double), Translation="0,0,0", Rotation="20,0,10", Scale="1,1,1") {
    rotationMode=BMM_Additive
    alphaBoolBlend {blendOption=Linear}
    alphaScaleBiasClamp {bClampResult=true, bInterpResult=true, bias=0.1, interpSpeedDecreasing=2, interpSpeedIncreasing=2}
    boneToModify {boneName=dyn_UFO_rotator_jnt}
    rotation {pitch=20, roll=10}
   }
   n10 = "Transform (Modify) Bone - Bone: dyn_planetCenter_jnt"(ComponentPose=@n9, Alpha=1, Translation="0,0,0", Rotation=Get rotateMainPlanet() -> (rotateMainPlanet:Rotator), Scale="1,1,1") {
    rotationMode=BMM_Additive
    alphaBoolBlend {blendOption=Linear}
    alphaScaleBiasClamp {interpSpeedDecreasing=2, interpSpeedIncreasing=2}
    boneToModify {boneName=dyn_planetCenter_jnt}
    rotation {pitch=20, roll=10}
   }
   Output Pose(Result=Component To Local(ComponentPose=@n10)) {layerGroup=None}
   n11 = Anim Dynamics - Bone: dyn_planetA_jnt!(ComponentPose=Local To Component(LocalPose=Input Pose())) -> (Pose) {
    alpha=0.4
    bGravityOverrideInSimSpace=true
    bLinearSpring=true
    linearSpringConstant=10
    simulationSpace=World
    alphaBoolBlend {blendOption=Linear}
    alphaScaleBiasClamp {interpSpeedDecreasing=5, interpSpeedIncreasing=5, scale=0.08}
    boundBone {boneName=dyn_planetA_jnt}
    componentLinearAccScale=<0.5 0.5 0.05>
    componentLinearVelScale=<0.1 0.1 0.12>
    physicsBodyDefinitions [{
      sphereCollisionRadius=0
      boundBone {boneName=dyn_planetA_jnt}
      boxExtents=<1 1 1>
      constraintSetup {angularLimitsMax=<0 0 0>, angularLimitsMin=<0 0 0>, angularTarget=<0 0 0>, linearAxesMax=<20 20 200>, linearAxesMin=<-20 -20 -200>}
      localJointOffset=<0 0 0>
     }]
   }
   n12 = Anim Dynamics - Bone: dyn_planetB_jnt!(ComponentPose=@n11) -> (Pose) {
    alpha=0.45
    bGravityOverrideInSimSpace=true
    bLinearSpring=true
    gravityScale=0.5
    linearSpringConstant=30
    simulationSpace=World
    alphaBoolBlend {blendOption=Linear}
    alphaScaleBiasClamp {interpSpeedDecreasing=5, interpSpeedIncreasing=5, scale=0.08}
    boundBone {boneName=dyn_planetB_jnt}
    componentLinearAccScale=<0.5 0.5 0.05>
    componentLinearVelScale=<0.1 0.1 0.12>
    physicsBodyDefinitions [{
      sphereCollisionRadius=0
      boundBone {boneName=dyn_planetB_jnt}
      boxExtents=<1 1 1>
      constraintSetup {angularLimitsMax=<0 0 0>, angularLimitsMin=<0 0 0>, angularTarget=<0 0 0>, linearAxesMax=<20 20 60>, linearAxesMin=<-20 -20 -190>}
      localJointOffset=<0 0 0>
     }]
   }
   n13 = Anim Dynamics - Bone: dyn_planetC_jnt!(ComponentPose=@n12) -> (Pose) {
    alpha=0.4
    bGravityOverrideInSimSpace=true
    bLinearSpring=true
    gravityScale=0.8
    linearSpringConstant=10
    simulationSpace=World
    alphaBoolBlend {blendOption=Linear}
    alphaScaleBiasClamp {interpSpeedDecreasing=5, interpSpeedIncreasing=5, scale=0.08}
    boundBone {boneName=dyn_planetC_jnt}
    componentLinearAccScale=<0.5 0.5 0.05>
    componentLinearVelScale=<0.1 0.1 0.12>
    physicsBodyDefinitions [{
      sphereCollisionRadius=0
      boundBone {boneName=dyn_planetC_jnt}
      boxExtents=<1 1 1>
      constraintSetup {angularLimitsMax=<0 0 0>, angularLimitsMin=<0 0 0>, angularTarget=<0 0 0>, linearAxesMax=<20 20 200>, linearAxesMin=<-20 -20 -200>}
      localJointOffset=<0 0 0>
     }]
   }
   n14 = "Transform (Modify) Bone - Bone: dyn_UFO_rotator_jnt"(ComponentPose=@n13, Alpha=1, Translation="0,0,0", Rotation="-50,0,-40", Scale="1,1,1") {
    rotationMode=BMM_Additive
    rotationSpace=BCS_BoneSpace
    translationSpace=BCS_BoneSpace
    alphaBoolBlend {blendOption=Linear}
    boneToModify {boneName=dyn_UFO_rotator_jnt}
    rotation {pitch=-50, roll=-40}
   }
   Save cached pose 'PlanetRotations'(Pose=Component To Local(ComponentPose=@n14)) {cacheName=PlanetRotations}
  ]}
 EventGraph {nodes [
   Event BlueprintUpdateAnimation()
   > n1 = Sequence!(execute) -> (then_0, then_1, then_2:Exec, then_3:Exec, then_4:Exec, then_5:Exec, then_6:Exec, then_7:Exec, then_8:Exec, then_9:Exec)
   n2 = "Sin (Radians)"(A=float * float(A=GetRealTimeSeconds(), B=2))
   > Set TranslateUFO(TranslateUFO:Vector=MakeVector(X=float * float(A=@n2, B=0.8), Y=0)) -> (Output_Get:Vector)
   n3 = float / float(A=GetRealTimeSeconds(), B=250) {is=title K2Node_CallFunction}
   Set RotatePlanet1(execute=@n1.then_1, RotatePlanet1:Rotator=MakeRotator(Yaw=float * float(A=@n3, B=10000))) -> (Output_Get:Rotator)
   n4 = float / float(A=GetRealTimeSeconds(), B=400)
   Set RotatePlanet2(execute=@n1.then_2, RotatePlanet2:Rotator=MakeRotator(Yaw=float * float(A=@n4, B=10000))) -> (Output_Get:Rotator)
   n5 = float / float(A=GetRealTimeSeconds(), B=550)
   Set RotatePlanet3(execute=@n1.then_3, RotatePlanet3:Rotator=MakeRotator(Yaw=float * float(A=@n5, B=10000))) -> (Output_Get:Rotator)
   n6 = GetRealTimeSeconds()
   n7 = float * float(A="Sin (Radians)"(A=float * float(A=@n6, B=5)), B=1.5)
   Set UFObanking(execute=@n1.then_4, UFObanking:Rotator=MakeRotator(Pitch=@n7, Yaw=@n7)) -> (Output_Get:Rotator)
   n8 = float / float(A=GetRealTimeSeconds(), B=70)
   Set RotateUFO(execute=@n1.then_5, RotateUFO:Rotator=MakeRotator(Roll=float * float(A=@n8, B=10000))) -> (Output_Get:Rotator)
   Delay(execute=@n1.then_6, Duration=2.7)
   n9 = RandomIntegerInRange!(Min:int=3, Max:int=5) -> (ReturnValue:int=0)
   > Set UFOpose1(UFOpose1:double="To Float (Boolean)"(InBool="Equal (Integer)"(A=@n9, B=3) {is=title K2Node_CallFunction})) -> (Output_Get:double)
   Delay(execute=@n1.then_7, Duration=2.3)
   > Set UFOpose2(UFOpose2:double="To Float (Boolean)"(InBool="Equal (Integer)"(A=@n9, B=4))) -> (Output_Get:double)
   Delay(execute=@n1.then_8, Duration=3)
   > Set UFOpose3(UFOpose3:double="To Float (Boolean)"(InBool="Equal (Integer)"(A=@n9, B=5))) -> (Output_Get:double)
   n10 = float / float(A=GetRealTimeSeconds(), B=325)
   Set rotateMainPlanet(execute=@n1.then_9, rotateMainPlanet:Rotator=MakeRotator(Yaw=float * float(A=@n10, B=-10000))) -> (Output_Get:Rotator)
   TryGetPawnOwner()
  ]}
]
variables [
 rotateMainPlanet {type=Rotator}
 RotatePlanet1 {type=Rotator}
 RotatePlanet2 {type=Rotator}
 RotatePlanet3 {type=Rotator}
 RotateUFO {type=Rotator}
 TranslateUFO {type=Vector}
 UFObanking {type=Rotator}
 UFOpose1 {type=number}
 UFOpose2 {type=number}
 UFOpose3 {type=number}
]
